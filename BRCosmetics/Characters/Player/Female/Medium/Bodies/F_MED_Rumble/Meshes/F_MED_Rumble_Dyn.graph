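# Physics asset for secondary motion of the skirt, shirt and scarf bones on the F_MED_Rumble female body.
# Capsules on pelvis, spine and thighs join small anchor capsules through free-angular constraints, and clothing segments hang from those anchors with mostly tight angular limits.
# Six three-segment skirt chains root to the pelvis, except right front and right side on thigh_r.
# Eight single-segment shirt flaps and a two-segment front scarf root to spine_03.
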PhysicsAsset /BRCosmetics/Characters/Player/Female/Medium/Bodies/F_MED_Rumble/Meshes/F_MED_Rumble_Dyn
bodies {
 dyn_scarf_fr_01 [{shapeType=Capsule, length=1.875, radius=6.5, shapeName=dyn_jacket_l_fr_02_capsule, center=<6 0 2>, rotation=<rot 90 0 0>}]
 dyn_scarf_fr_02 [{shapeType=Capsule, length=2.25, radius=6.125, shapeName=dyn_jacket_l_fr_02_capsule, center=<6 0 2>, rotation=<rot 90 0 0>}]
 dyn_scarf_fr_MSTR [{shapeType=Capsule, length=1, radius=1}]
 dyn_shirt_bk [{shapeType=Capsule, length=4.875, radius=3.5, shapeName=dyn_jacket_l_fr_02_capsule, center=<8 0 2>, rotation=<rot 0 0 84.9986>}]
 dyn_shirt_bk_MSTR [{shapeType=Capsule, length=1, radius=1}]
 dyn_shirt_bk_l [{shapeType=Capsule, length=4.875, radius=3.5, shapeName=dyn_jacket_l_fr_02_capsule, center=<8 0 2>, rotation=<rot 0 0 84.9986>}]
 dyn_shirt_bk_l_MSTR [{shapeType=Capsule, length=1, radius=1}]
 dyn_shirt_bk_r [{shapeType=Capsule, length=4.875, radius=3.5, shapeName=dyn_jacket_l_fr_02_capsule, center=<8 0 2>, rotation=<rot 0 0 84.9986>}]
 dyn_shirt_bk_r_MSTR [{shapeType=Capsule, length=1, radius=1}]
 dyn_shirt_fr [{shapeType=Capsule, length=4.75, radius=3.625, shapeName=dyn_jacket_l_fr_02_capsule, center=<5 0 2>, rotation=<rot 0 0 84.9986>}]
 dyn_shirt_fr_MSTR [{shapeType=Capsule, length=1, radius=1}]
 dyn_shirt_fr_l [{shapeType=Capsule, length=4.75, radius=3.625, shapeName=dyn_jacket_l_fr_02_capsule, center=<5 0 2>, rotation=<rot 0 0 84.9986>}]
 dyn_shirt_fr_l_MSTR [{shapeType=Capsule, length=1, radius=1}]
 dyn_shirt_fr_r [{shapeType=Capsule, length=4.75, radius=3.625, shapeName=dyn_jacket_l_fr_02_capsule, center=<5 0 2>, rotation=<rot 0 0 84.9986>}]
 dyn_shirt_fr_r_MSTR [{shapeType=Capsule, length=1, radius=1}]
 dyn_shirt_l [{shapeType=Capsule, length=5.25, radius=3.125, shapeName=dyn_jacket_l_fr_02_capsule, center=<5 0 2>, rotation=<rot 0 0 84.9986>}]
 dyn_shirt_l_MSTR [{shapeType=Capsule, length=1, radius=1}]
 dyn_shirt_r [{shapeType=Capsule, length=5.25, radius=3.125, shapeName=dyn_jacket_l_fr_02_capsule, center=<5 0 2>, rotation=<rot 0 0 84.9986>}]
 dyn_shirt_r_MSTR [{shapeType=Capsule, length=1, radius=1}]
 dyn_skirt_bk_a_base_l [{shapeType=Capsule, length=1, radius=1}]
 dyn_skirt_bk_a_base_r [{shapeType=Capsule, length=1, radius=1}]
 dyn_skirt_bk_a_l [{shapeType=Capsule, length=9.75, radius=4.5, center=<4 0 4>, rotation=<rot 0 0 90>}]
 dyn_skirt_bk_a_r [{shapeType=Capsule, length=9.75, radius=4.5, center=<4 0 4>, rotation=<rot 0 0 90>}]
 dyn_skirt_bk_b_l [{shapeType=Capsule, length=14.125, radius=6.5, center=<4 0 4>, rotation=<rot 0 0 85>}]
 dyn_skirt_bk_b_r [{shapeType=Capsule, length=14.125, radius=6.5, center=<4 0 4>, rotation=<rot -0.0001 0 95>}]
 dyn_skirt_bk_c_l [{shapeType=Capsule, length=17.25, radius=7.875, center=<4 0 4>, rotation=<rot 0 0 90>}]
 dyn_skirt_bk_c_r [{shapeType=Capsule, length=17.25, radius=7.875, center=<4 0 4>, rotation=<rot 0 0 90>}]
 dyn_skirt_fr_a_base_l [{shapeType=Capsule, length=1, radius=1}]
 dyn_skirt_fr_a_base_r [{shapeType=Capsule, length=1, radius=1}]
 dyn_skirt_fr_a_l [{shapeType=Capsule, length=12.25, radius=5.375, center=<5 0 4>, rotation=<rot 0.0007 0.0002 89.9998>}]
 dyn_skirt_fr_a_r [{shapeType=Capsule, length=10.625, radius=5.375, center=<5 0 4>, rotation=<rot 0.0005 -0.0001 100.0001>}]
 dyn_skirt_fr_b_l [{shapeType=Capsule, length=15.75, radius=6.75, center=<5 -3 4>, rotation=<rot 0.0009 0.0002 74.9998>}]
 dyn_skirt_fr_b_r [{shapeType=Capsule, length=14.75, radius=6.125, center=<5 3 4>, rotation=<rot 0.0005 0 120.0001>}]
 dyn_skirt_fr_c_l [{shapeType=Capsule, length=19.125, radius=7.625, center=<5 -3 4>, rotation=<rot 0.0014 0.0003 64.9997>}]
 dyn_skirt_fr_c_r [{shapeType=Capsule, length=17.375, radius=7.625, center=<5 3 4>, rotation=<rot 0.0014 0.0002 120>}]
 dyn_skirt_side_a_base_l [{shapeType=Capsule, length=1, radius=1}]
 dyn_skirt_side_a_base_r [{shapeType=Capsule, length=1, radius=1}]
 dyn_skirt_side_a_l [{shapeType=Capsule, length=17.375, radius=5.25, center=<4 0 4>, rotation=<rot 75.0001 89.9999 0>}]
 dyn_skirt_side_a_r [{shapeType=Capsule, length=8.125, radius=5.25, center=<4 0 4>, rotation=<rot 90 -90 -179.9997>}]
 dyn_skirt_side_b_l [{shapeType=Capsule, length=25.125, radius=4.75, center=<4 0 4>, rotation=<rot 75.0001 90.0029 0>}]
 dyn_skirt_side_b_r [{shapeType=Capsule, length=15.5, radius=5.125, center=<4 0 4>, rotation=<rot 90 -14.0363 -104.039>}]
 dyn_skirt_side_c_l [{shapeType=Capsule, length=23.75, radius=6.5, center=<4 0 4>, rotation=<rot 74.2069 71.9823 -18.6788>}]
 dyn_skirt_side_c_r [{shapeType=Capsule, length=15.625, radius=6.25, center=<4 0 4>, rotation=<rot 84.9995 -0.0003 -90.0025>}]
 pelvis [{shapeType=Capsule, length=9.25, radius=10.25, center=<-1.7233 -1.7603 0>}]
 spine_02 [{shapeType=Capsule, length=4.7202, radius=7.669, center=<0.8218 1.7395 0>}]
 spine_03 [{shapeType=Capsule, length=3.9702, radius=7.919, center=<-0.0001 1.7395 0>}]
 thigh_l [{shapeType=Capsule, length=28.5, radius=9.875, center=<-17.8489 -0.7558 1.0679>, rotation=<rot 88.0783 179.9965 174.8622>}]
 thigh_r [{shapeType=Capsule, length=28.5, radius=9.875, center=<17.8489 0.7558 -1.0679>, rotation=<rot 88.0783 179.9965 -5.1378>}]
}
constraints [
 {
  bone1Name=dyn_skirt_side_a_base_l
  bone2Name=pelvis
  swing1LimitDegrees=45
  swing1Motion=Free
  swing2LimitDegrees=45
  swing2Motion=Free
  twistLimitDegrees=45
  twistMotion=Free
 }
 {
  bone1Name=dyn_skirt_bk_a_base_l
  bone2Name=pelvis
  swing1LimitDegrees=45
  swing1Motion=Free
  swing2LimitDegrees=45
  swing2Motion=Free
  twistLimitDegrees=45
  twistMotion=Free
 }
 {
  bone1Name=dyn_skirt_fr_a_base_l
  bone2Name=pelvis
  swing1LimitDegrees=45
  swing1Motion=Free
  swing2LimitDegrees=45
  swing2Motion=Free
  twistLimitDegrees=45
  twistMotion=Free
 }
 {
  bone1Name=dyn_skirt_side_a_base_r
  bone2Name=thigh_r
  swing1LimitDegrees=45
  swing1Motion=Free
  swing2LimitDegrees=45
  swing2Motion=Free
  twistLimitDegrees=45
  twistMotion=Free
 }
 {
  bone1Name=dyn_skirt_bk_a_base_r
  bone2Name=pelvis
  swing1LimitDegrees=45
  swing1Motion=Free
  swing2LimitDegrees=45
  swing2Motion=Free
  twistLimitDegrees=45
  twistMotion=Free
 }
 {
  bone1Name=dyn_skirt_fr_a_base_r
  bone2Name=thigh_r
  swing1LimitDegrees=45
  swing1Motion=Free
  swing2LimitDegrees=45
  swing2Motion=Free
  twistLimitDegrees=45
  twistMotion=Free
 }
 {bone1Name=dyn_skirt_side_a_l, bone2Name=dyn_skirt_side_a_base_l, swing1LimitDegrees=2, swing2LimitDegrees=15, twistLimitDegrees=1}
 {bone1Name=dyn_skirt_side_b_l, bone2Name=dyn_skirt_side_a_l, swing1LimitDegrees=2, swing2LimitDegrees=15, twistLimitDegrees=1}
 {bone1Name=dyn_skirt_bk_b_l, bone2Name=dyn_skirt_bk_a_l, swing1LimitDegrees=2, swing2LimitDegrees=25, twistLimitDegrees=1}
 {bone1Name=dyn_skirt_bk_a_l, bone2Name=dyn_skirt_bk_a_base_l, swing1LimitDegrees=2, swing2LimitDegrees=15, twistLimitDegrees=1}
 {
  bone1Name=dyn_skirt_fr_a_l
  bone2Name=dyn_skirt_fr_a_base_l
  swing1LimitDegrees=2
  swing2LimitDegrees=10
  twistLimitDegrees=1
  twistMotion=Locked
 }
 {bone1Name=dyn_skirt_fr_b_l, bone2Name=dyn_skirt_fr_a_l, swing1LimitDegrees=2, swing2LimitDegrees=10, twistLimitDegrees=1, twistMotion=Locked}
 {bone1Name=dyn_skirt_fr_a_r, bone2Name=dyn_skirt_fr_a_base_r, swing1LimitDegrees=2, swing2LimitDegrees=15, twistLimitDegrees=1}
 {bone1Name=dyn_skirt_side_a_r, bone2Name=dyn_skirt_side_a_base_r, swing1LimitDegrees=2, swing2LimitDegrees=15, twistLimitDegrees=1}
 {bone1Name=dyn_skirt_bk_a_r, bone2Name=dyn_skirt_bk_a_base_r, swing1LimitDegrees=2, swing2LimitDegrees=15, twistLimitDegrees=1}
 {bone1Name=dyn_skirt_bk_b_r, bone2Name=dyn_skirt_bk_a_r, swing1LimitDegrees=2, swing2LimitDegrees=25, twistLimitDegrees=1}
 {bone1Name=dyn_skirt_fr_b_r, bone2Name=dyn_skirt_fr_a_r, swing1LimitDegrees=2, swing2LimitDegrees=15, twistLimitDegrees=1}
 {bone1Name=dyn_skirt_side_b_r, bone2Name=dyn_skirt_side_a_r, swing1LimitDegrees=2, swing2LimitDegrees=15, twistLimitDegrees=1}
 {bone1Name=dyn_skirt_side_c_l, bone2Name=dyn_skirt_side_b_l, swing1LimitDegrees=2, swing2LimitDegrees=15, twistLimitDegrees=1}
 {bone1Name=dyn_skirt_bk_c_l, bone2Name=dyn_skirt_bk_b_l, swing1LimitDegrees=2, swing2LimitDegrees=25, twistLimitDegrees=1}
 {bone1Name=dyn_skirt_fr_c_l, bone2Name=dyn_skirt_fr_b_l, swing1LimitDegrees=2, swing2LimitDegrees=10, twistLimitDegrees=1, twistMotion=Locked}
 {bone1Name=dyn_skirt_side_c_r, bone2Name=dyn_skirt_side_b_r, swing1LimitDegrees=2, swing2LimitDegrees=15, twistLimitDegrees=1}
 {bone1Name=dyn_skirt_bk_c_r, bone2Name=dyn_skirt_bk_b_r, swing1LimitDegrees=2, swing2LimitDegrees=25, twistLimitDegrees=1}
 {bone1Name=dyn_skirt_fr_c_r, bone2Name=dyn_skirt_fr_b_r, swing1LimitDegrees=2, swing2LimitDegrees=15, twistLimitDegrees=1}
 {
  bone1Name=dyn_shirt_fr_MSTR
  bone2Name=spine_03
  swing1LimitDegrees=45
  swing1Motion=Free
  swing2LimitDegrees=45
  swing2Motion=Free
  twistLimitDegrees=45
  twistMotion=Free
 }
 {bone1Name=dyn_shirt_fr, bone2Name=dyn_shirt_fr_MSTR, swing1LimitDegrees=10, swing2LimitDegrees=15, twistLimitDegrees=5}
 {
  bone1Name=dyn_shirt_fr_l_MSTR
  bone2Name=spine_03
  swing1LimitDegrees=45
  swing1Motion=Free
  swing2LimitDegrees=45
  swing2Motion=Free
  twistLimitDegrees=45
  twistMotion=Free
 }
 {bone1Name=dyn_shirt_fr_l, bone2Name=dyn_shirt_fr_l_MSTR, swing1LimitDegrees=10, swing2LimitDegrees=15, twistLimitDegrees=5}
 {
  bone1Name=dyn_shirt_l_MSTR
  bone2Name=spine_03
  swing1LimitDegrees=45
  swing1Motion=Free
  swing2LimitDegrees=45
  swing2Motion=Free
  twistLimitDegrees=45
  twistMotion=Free
 }
 {bone1Name=dyn_shirt_l, bone2Name=dyn_shirt_l_MSTR, swing1LimitDegrees=20, swing2LimitDegrees=20, twistLimitDegrees=5}
 {
  bone1Name=dyn_shirt_bk_l_MSTR
  bone2Name=spine_03
  swing1LimitDegrees=45
  swing1Motion=Free
  swing2LimitDegrees=45
  swing2Motion=Free
  twistLimitDegrees=45
  twistMotion=Free
 }
 {bone1Name=dyn_shirt_bk_l, bone2Name=dyn_shirt_bk_l_MSTR, swing1LimitDegrees=10, swing2LimitDegrees=20, twistLimitDegrees=5}
 {
  bone1Name=dyn_shirt_bk_MSTR
  bone2Name=spine_03
  swing1LimitDegrees=45
  swing1Motion=Free
  swing2LimitDegrees=45
  swing2Motion=Free
  twistLimitDegrees=45
  twistMotion=Free
 }
 {bone1Name=dyn_shirt_bk, bone2Name=dyn_shirt_bk_MSTR, swing1LimitDegrees=10, swing2LimitDegrees=20, twistLimitDegrees=5}
 {
  bone1Name=dyn_shirt_fr_r_MSTR
  bone2Name=spine_03
  swing1LimitDegrees=45
  swing1Motion=Free
  swing2LimitDegrees=45
  swing2Motion=Free
  twistLimitDegrees=45
  twistMotion=Free
 }
 {bone1Name=dyn_shirt_fr_r, bone2Name=dyn_shirt_fr_r_MSTR, swing1LimitDegrees=10, swing2LimitDegrees=15, twistLimitDegrees=5}
 {
  bone1Name=dyn_shirt_r_MSTR
  bone2Name=spine_03
  swing1LimitDegrees=45
  swing1Motion=Free
  swing2LimitDegrees=45
  swing2Motion=Free
  twistLimitDegrees=45
  twistMotion=Free
 }
 {bone1Name=dyn_shirt_r, bone2Name=dyn_shirt_r_MSTR, swing1LimitDegrees=20, swing2LimitDegrees=20, twistLimitDegrees=5}
 {
  bone1Name=dyn_shirt_bk_r_MSTR
  bone2Name=spine_03
  swing1LimitDegrees=45
  swing1Motion=Free
  swing2LimitDegrees=45
  swing2Motion=Free
  twistLimitDegrees=45
  twistMotion=Free
 }
 {bone1Name=dyn_shirt_bk_r, bone2Name=dyn_shirt_bk_r_MSTR, swing1LimitDegrees=10, swing2LimitDegrees=20, twistLimitDegrees=5}
 {
  bone1Name=spine_03
  bone2Name=spine_02
  swing1LimitDegrees=45
  swing1Motion=Free
  swing2LimitDegrees=45
  swing2Motion=Free
  twistLimitDegrees=45
  twistMotion=Free
 }
 {
  bone1Name=thigh_l
  bone2Name=pelvis
  swing1LimitDegrees=45
  swing1Motion=Free
  swing2LimitDegrees=45
  swing2Motion=Free
  twistLimitDegrees=45
  twistMotion=Free
 }
 {
  bone1Name=spine_02
  bone2Name=pelvis
  swing1LimitDegrees=45
  swing1Motion=Free
  swing2LimitDegrees=45
  swing2Motion=Free
  twistLimitDegrees=45
  twistMotion=Free
 }
 {
  bone1Name=dyn_scarf_fr_MSTR
  bone2Name=spine_03
  swing1LimitDegrees=45
  swing1Motion=Free
  swing2LimitDegrees=45
  swing2Motion=Free
  twistLimitDegrees=45
  twistMotion=Free
 }
 {bone1Name=dyn_scarf_fr_01, bone2Name=dyn_scarf_fr_MSTR, swing1LimitDegrees=45, swing2LimitDegrees=35, twistLimitDegrees=2}
 {bone1Name=dyn_scarf_fr_02, bone2Name=dyn_scarf_fr_01, swing1LimitDegrees=45, swing2LimitDegrees=35, twistLimitDegrees=2}
 {
  bone1Name=thigh_r
  bone2Name=pelvis
  swing1LimitDegrees=45
  swing1Motion=Free
  swing2LimitDegrees=45
  swing2Motion=Free
  twistLimitDegrees=45
  twistMotion=Free
 }
]
