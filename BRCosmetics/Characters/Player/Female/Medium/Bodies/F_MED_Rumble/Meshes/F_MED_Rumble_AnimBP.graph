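# Anim Blueprint for the F_MED_Rumble female body cosmetic, copying the parent mesh's pose and adding deformation and secondary cloth motion.
# Two deform Control Rigs feed scarf, shirt and skirt stages using Look Ats, AnimDynamics and RigidBody physics.
# Each update maps thigh and left-arm rotations to corrective bone-offset alphas, applied before and after RigidBody.
# A backbling check, made once when customization loads, swaps back-scarf dynamics for fixed tilts.
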AnimBlueprint /BRCosmetics/Characters/Player/Female/Medium/Bodies/F_MED_Rumble/Meshes/F_MED_Rumble_AnimBP
target_skeleton=&/F_MED_Rumble_Skeleton
graphs [
 AnimGraph {nodes [
   n1 = "Transform (Modify) Bone - Bone: dyn_skirt_fr_a_base_l"(Alpha=0, Translation="0,0,0", Rotation="0,15,10", Scale="1,1,1") {
    rotationMode=BMM_Additive
    rotationSpace=BCS_BoneSpace
    translationMode=BMM_Additive
    translationSpace=BCS_BoneSpace
    alphaBoolBlend {blendOption=Linear}
    boneToModify {boneName=dyn_skirt_fr_a_base_l}
    rotation {roll=10, yaw=15}
   }
   "Transform (Modify) Bone - Bone: dyn_skirt_side_a_base_l"(ComponentPose=@n1, Alpha=0, Translation="0,0,0", Rotation="10,0,0", Scale="1,1,1") {
    rotationMode=BMM_Additive
    rotationSpace=BCS_BoneSpace
    translationMode=BMM_Additive
    translationSpace=BCS_BoneSpace
    alphaBoolBlend {blendOption=Linear}
    boneToModify {boneName=dyn_skirt_side_a_base_l}
    rotation {pitch=10}
   }
   "Transform (Modify) Bone - Bone: dyn_skirt_fr_a_base_l"(Alpha=Get thigh_rot_otr_l() -> (thigh_rot_otr_l:double), Translation="0,0,0", Rotation="15,0,0", Scale="1,1,1") {
    rotationMode=BMM_Additive
    rotationSpace=BCS_BoneSpace
    translationMode=BMM_Additive
    translationSpace=BCS_BoneSpace
    alphaBoolBlend {blendOption=Linear}
    boneToModify {boneName=dyn_skirt_fr_a_base_l}
    rotation {pitch=15}
   }
   n2 = Control Rig(Source=Copy Pose From Mesh() {bUseAttachedParent=true}) {
    alphaBoolBlend {blendOption=Linear}
    controlRigAssetReference {blueprintRigClass=&/F_MED_Rumble_Deform_CtrlRig_twists.F_MED_Rumble_Deform_CtrlRig_twists_C}
   }
   n3 = Control Rig!(
    Source=@n2
    Alpha
    KneeRot_fr_l_114deg:Vector
    KneeTrans_fr_l_114deg:Vector="0.258,-3.591,0.009"
    KneeScale_fr_l_114deg:Vector="1,1,1"
    KneeScale_fr_l_130deg:Vector="1,1,1"
    KneeRot_fr_l_130deg:Vector="0,0,-51.2"
    KneeTrans_fr_l_130deg:Vector="0.258,-3.591,0.009"
   ) -> (Pose) {
    alphaBoolBlend {blendOption=Linear}
    controlRigAssetReference {blueprintRigClass=&/F_MED_Rumble_Deform_CtrlRig_sculpters.F_MED_Rumble_Deform_CtrlRig_sculpters_C}
   }
   n4 = "Transform (Modify) Bone - Bone: def_scarf_l"(
    ComponentPose=Local To Component(LocalPose=@n3)
    Alpha=Get arm_rot_up_l() -> (arm_rot_up_l:double)
    Translation="0,0,0"
    Rotation="30,0,0"
    Scale="0.8,1,1"
   ) {
    rotationMode=BMM_Additive
    rotationSpace=BCS_BoneSpace
    scaleMode=BMM_Additive
    scaleSpace=BCS_BoneSpace
    alphaBoolBlend {blendOption=Linear}
    boneToModify {boneName=def_scarf_l}
    rotation {pitch=30}
    scale {x=0.8}
   }
   n5 = Get isWearBackbling() -> (isWearBackbling:bool)
   n6 = "Transform (Modify) Bone - Bone: dyn_scarf_bk_MSTR"!(ComponentPose=@n4, bAlphaBoolEnabled:bool=@n5, Translation="0,0,0", Rotation="-12,0,0", Scale="1,1,1") -> (Pose) {
    alphaInputType=Bool
    rotationMode=BMM_Additive
    rotationSpace=BCS_BoneSpace
    scaleMode=BMM_Additive
    scaleSpace=BCS_BoneSpace
    alphaBoolBlend {blendOption=Linear}
    boneToModify {boneName=dyn_scarf_bk_MSTR}
    rotation {pitch=-12}
   }
   n7 = "Transform (Modify) Bone - Bone: dyn_scarf_bk_l_MSTR"!(ComponentPose=@n6, bAlphaBoolEnabled:bool=@n5, Translation="0,0,0", Rotation="-6,0,0", Scale="1,1,1") -> (Pose) {
    alphaInputType=Bool
    rotationMode=BMM_Additive
    rotationSpace=BCS_BoneSpace
    scaleMode=BMM_Additive
    scaleSpace=BCS_BoneSpace
    alphaBoolBlend {blendOption=Linear}
    boneToModify {boneName=dyn_scarf_bk_l_MSTR}
    rotation {pitch=-6}
   }
   n8 = "Transform (Modify) Bone - Bone: dyn_scarf_bk_r_MSTR"!(ComponentPose=@n7, bAlphaBoolEnabled:bool=@n5, Translation="0,0,0", Rotation="-6,-0.000,-0.000", Scale="1,1,1") -> (Pose) {
    alphaInputType=Bool
    rotationMode=BMM_Additive
    rotationSpace=BCS_BoneSpace
    scaleMode=BMM_Additive
    scaleSpace=BCS_BoneSpace
    alphaBoolBlend {blendOption=Linear}
    boneToModify {boneName=dyn_scarf_bk_r_MSTR}
    rotation {pitch=-6}
   }
   n9 = NOT Boolean(A=Get isWearBackbling())
   n10 = Anim Dynamics - Bone: dyn_scarf_bk!(ComponentPose=@n8, bAlphaBoolEnabled:bool=@n9, ExternalForce="0,0,0") -> (Pose) {
    alphaInputType=Bool
    angularDampingOverride=0.99
    bOverrideAngularDamping=true
    bUseGravityOverride=true
    simulationSpace=RootRelative
    alphaBoolBlend {blendOption=Linear}
    boundBone {boneName=dyn_scarf_bk}
    componentAppliedLinearAccClamp=<5000 5000 5000>
    componentLinearAccScale=<0.1 0.35 0.35>
    componentLinearVelScale=<0.1 0.35 0.35>
    gravityOverride {z=-980}
    physicsBodyDefinitions [{
      boundBone {boneName=dyn_scarf_bk}
      boxExtents=<10 30 15>
      constraintSetup {angularLimitsMax=<0 10 5>, angularLimitsMin=<0 -60 -5>, angularTarget=<0 0 0>}
      localJointOffset=<3 0 0>
     }]
   }
   n11 = Anim Dynamics - Bone: dyn_scarf_bk_l!(ComponentPose=@n10, bAlphaBoolEnabled:bool=@n9, ExternalForce="0,0,0") -> (Pose) {
    alphaInputType=Bool
    angularDampingOverride=0.99
    bOverrideAngularDamping=true
    bUseGravityOverride=true
    simulationSpace=RootRelative
    alphaBoolBlend {blendOption=Linear}
    boundBone {boneName=dyn_scarf_bk_l}
    componentAppliedLinearAccClamp=<5000 5000 5000>
    componentLinearAccScale=<0.1 0.35 0.35>
    componentLinearVelScale=<0.1 0.35 0.35>
    gravityOverride {z=-980}
    physicsBodyDefinitions [{
      boundBone {boneName=dyn_scarf_bk_l}
      boxExtents=<10 30 15>
      constraintSetup {angularLimitsMax=<0 5 5>, angularLimitsMin=<0 -60 -5>, angularTarget=<0 0 0>}
      localJointOffset=<3 0 0>
     }]
   }
   n12 = Anim Dynamics - Bone: dyn_scarf_bk_r!(ComponentPose=@n11, bAlphaBoolEnabled:bool=@n9, ExternalForce="0,0,0") -> (Pose) {
    alphaInputType=Bool
    angularDampingOverride=0.99
    bOverrideAngularDamping=true
    bUseGravityOverride=true
    simulationSpace=RootRelative
    alphaBoolBlend {blendOption=Linear}
    boundBone {boneName=dyn_scarf_bk_r}
    componentAppliedLinearAccClamp=<5000 5000 5000>
    componentLinearAccScale=<0.1 0.35 0.35>
    componentLinearVelScale=<0.1 0.35 0.35>
    gravityOverride {z=-980}
    physicsBodyDefinitions [{
      boundBone {boneName=dyn_scarf_bk_r}
      boxExtents=<10 30 15>
      constraintSetup {angularLimitsMax=<0 5 5>, angularLimitsMin=<0 -60 -5>, angularTarget=<0 0 0>}
      localJointOffset=<3 0 0>
     }]
   }
   n13 = "Constraint - dyn_scarf_fr_MSTR (PP)"!(ComponentPose=@n12, Alpha=1, ConstraintWeights_0=0.7647, ConstraintWeights_1:float=0.2353) -> (Pose) {
    alphaBoolBlend {blendOption=Linear}
    boneToModify {boneName=dyn_scarf_fr_MSTR}
    constraintSetup [{targetBone {boneName=spine_04}}, {targetBone {boneName=spine_05}}]
    constraintWeights [1,1]
   }
   n14 = Look At - Bone: dyn_shirt_fr_MSTR(ComponentPose=@n13, Alpha=1) {
    alphaBoolBlend {blendOption=Linear}
    boneToModify {boneName=dyn_shirt_fr_MSTR}
    lookAtLocation=<-10 10 0>
    lookAtTarget {boneReference {boneName=pelvis}}
    lookAt_Axis {axis=<1 0>}
   }
   n15 = Look At - Bone: dyn_shirt_fr_l_MSTR(ComponentPose=@n14, Alpha=1) {
    alphaBoolBlend {blendOption=Linear}
    boneToModify {boneName=dyn_shirt_fr_l_MSTR}
    lookAtLocation=<-10 8.1309 -11.6796>
    lookAtTarget {boneReference {boneName=pelvis}}
    lookAt_Axis {axis=<1 0>}
   }
   n16 = Look At - Bone: dyn_shirt_fr_r_MSTR(ComponentPose=@n15, Alpha=1) {
    alphaBoolBlend {blendOption=Linear}
    boneToModify {boneName=dyn_shirt_fr_r_MSTR}
    lookAtLocation=<-10 11.6027 11.7518>
    lookAtTarget {boneReference {boneName=spine_02}}
    lookAt_Axis {axis=<1 0>}
   }
   n17 = Look At - Bone: dyn_shirt_l_MSTR(ComponentPose=@n16, Alpha=1) {
    alphaBoolBlend {blendOption=Linear}
    boneToModify {boneName=dyn_shirt_l_MSTR}
    lookAtLocation=<-10 0 -14.3436>
    lookAtTarget {boneReference {boneName=pelvis}}
    lookAt_Axis {axis=<1 0>}
   }
   n18 = Look At - Bone: dyn_shirt_r_MSTR(ComponentPose=@n17, Alpha=1) {
    alphaBoolBlend {blendOption=Linear}
    boneToModify {boneName=dyn_shirt_r_MSTR}
    lookAtLocation=<-10 0 13.5668>
    lookAtTarget {boneReference {boneName=spine_01}}
    lookAt_Axis {axis=<1 0>}
   }
   n19 = Anim Dynamics - Bone: dyn_shirt_bk!(ComponentPose=@n18, bAlphaBoolEnabled:bool=True, ExternalForce="0,0,0") -> (Pose) {
    alphaInputType=Bool
    angularDampingOverride=0.99
    angularSpringConstant=1000
    bAngularSpring=true
    bOverrideAngularDamping=true
    bUseGravityOverride=true
    simulationSpace=RootRelative
    alphaBoolBlend {blendOption=Linear}
    boundBone {boneName=dyn_shirt_bk}
    componentAppliedLinearAccClamp=<5000 5000 5000>
    componentLinearAccScale=<0.1 0.35 0.35>
    componentLinearVelScale=<0.1 0.35 0.35>
    gravityOverride {z=-980}
    physicsBodyDefinitions [{
      boundBone {boneName=dyn_shirt_bk}
      boxExtents=<15 20 10>
      constraintSetup {angularLimitsMax=<0 15 10>, angularLimitsMin=<0 -20 -10>, angularTarget=<0 0 0>}
      localJointOffset=<2 0 0>
     }]
   }
   n20 = Anim Dynamics - Bone: dyn_shirt_bk_l!(ComponentPose=@n19, bAlphaBoolEnabled:bool=True, ExternalForce="0,0,0") -> (Pose) {
    alphaInputType=Bool
    angularDampingOverride=0.99
    angularSpringConstant=1000
    bAngularSpring=true
    bOverrideAngularDamping=true
    bUseGravityOverride=true
    simulationSpace=RootRelative
    alphaBoolBlend {blendOption=Linear}
    boundBone {boneName=dyn_shirt_bk_l}
    componentAppliedLinearAccClamp=<5000 5000 5000>
    componentLinearAccScale=<0.1 0.35 0.35>
    componentLinearVelScale=<0.1 0.35 0.35>
    gravityOverride {z=-980}
    physicsBodyDefinitions [{
      boundBone {boneName=dyn_shirt_bk_l}
      boxExtents=<15 20 10>
      constraintSetup {angularLimitsMax=<0 15 10>, angularLimitsMin=<0 -20 -10>, angularTarget=<0 0 0>}
      localJointOffset=<2 0 0>
     }]
   }
   n21 = Anim Dynamics - Bone: dyn_shirt_bk_r!(ComponentPose=@n20, bAlphaBoolEnabled:bool=True, ExternalForce="0,0,0") -> (Pose) {
    alphaInputType=Bool
    angularDampingOverride=0.99
    angularSpringConstant=1000
    bAngularSpring=true
    bOverrideAngularDamping=true
    bUseGravityOverride=true
    simulationSpace=RootRelative
    alphaBoolBlend {blendOption=Linear}
    boundBone {boneName=dyn_shirt_bk_r}
    componentAppliedLinearAccClamp=<5000 5000 5000>
    componentLinearAccScale=<0.1 0.35 0.35>
    componentLinearVelScale=<0.1 0.35 0.35>
    gravityOverride {z=-980}
    physicsBodyDefinitions [{
      boundBone {boneName=dyn_shirt_bk_r}
      boxExtents=<15 20 10>
      constraintSetup {angularLimitsMax=<0 15 10>, angularLimitsMin=<0 -20 -10>, angularTarget=<0 0 0>}
      localJointOffset=<2 0 0>
     }]
   }
   n22 = Anim Dynamics - Bone: dyn_shirt_fr!(ComponentPose=@n21, bAlphaBoolEnabled:bool=True, ExternalForce="0,0,0") -> (Pose) {
    alphaInputType=Bool
    angularDampingOverride=0.99
    angularSpringConstant=1000
    bAngularSpring=true
    bOverrideAngularDamping=true
    bUseGravityOverride=true
    simulationSpace=RootRelative
    alphaBoolBlend {blendOption=Linear}
    boundBone {boneName=dyn_shirt_fr}
    componentAppliedLinearAccClamp=<5000 5000 5000>
    componentLinearAccScale=<0.1 0.35 0.35>
    componentLinearVelScale=<0.1 0.35 0.35>
    gravityOverride {z=-980}
    physicsBodyDefinitions [{
      boundBone {boneName=dyn_shirt_fr}
      boxExtents=<15 20 10>
      constraintSetup {angularLimitsMax=<0 10 10>, angularLimitsMin=<0 -10 -10>, angularTarget=<0 0 0>}
      localJointOffset=<1.5 0 0>
     }]
   }
   n23 = Anim Dynamics - Bone: dyn_shirt_fr_l!(ComponentPose=@n22, bAlphaBoolEnabled:bool=True, ExternalForce="0,0,0") -> (Pose) {
    alphaInputType=Bool
    angularDampingOverride=0.99
    angularSpringConstant=1000
    bAngularSpring=true
    bOverrideAngularDamping=true
    bUseGravityOverride=true
    simulationSpace=RootRelative
    alphaBoolBlend {blendOption=Linear}
    boundBone {boneName=dyn_shirt_fr_l}
    componentAppliedLinearAccClamp=<5000 5000 5000>
    componentLinearAccScale=<0.1 0.35 0.35>
    componentLinearVelScale=<0.1 0.35 0.35>
    gravityOverride {z=-980}
    physicsBodyDefinitions [{
      boundBone {boneName=dyn_shirt_fr_l}
      boxExtents=<15 20 10>
      constraintSetup {angularLimitsMax=<0 10 10>, angularLimitsMin=<0 -10 -10>, angularTarget=<0 0 0>}
      localJointOffset=<1.5 0 0>
     }]
   }
   n24 = Anim Dynamics - Bone: dyn_shirt_fr_r!(ComponentPose=@n23, bAlphaBoolEnabled:bool=True, ExternalForce="0,0,0") -> (Pose) {
    alphaInputType=Bool
    angularDampingOverride=0.99
    angularSpringConstant=1000
    bAngularSpring=true
    bOverrideAngularDamping=true
    bUseGravityOverride=true
    simulationSpace=RootRelative
    alphaBoolBlend {blendOption=Linear}
    boundBone {boneName=dyn_shirt_fr_r}
    componentAppliedLinearAccClamp=<5000 5000 5000>
    componentLinearAccScale=<0.1 0.35 0.35>
    componentLinearVelScale=<0.1 0.35 0.35>
    gravityOverride {z=-980}
    physicsBodyDefinitions [{
      boundBone {boneName=dyn_shirt_fr_r}
      boxExtents=<15 20 10>
      constraintSetup {angularLimitsMax=<0 10 10>, angularLimitsMin=<0 -10 -10>, angularTarget=<0 0 0>}
      localJointOffset=<1.5 0 0>
     }]
   }
   n25 = Anim Dynamics - Bone: dyn_shirt_l!(ComponentPose=@n24, bAlphaBoolEnabled:bool=True, ExternalForce="0,0,0") -> (Pose) {
    alphaInputType=Bool
    angularDampingOverride=0.99
    angularSpringConstant=1000
    bAngularSpring=true
    bOverrideAngularDamping=true
    bUseGravityOverride=true
    simulationSpace=RootRelative
    alphaBoolBlend {blendOption=Linear}
    boundBone {boneName=dyn_shirt_l}
    componentAppliedLinearAccClamp=<5000 5000 5000>
    componentLinearAccScale=<0.1 0.35 0.35>
    componentLinearVelScale=<0.1 0.35 0.35>
    gravityOverride {z=-980}
    physicsBodyDefinitions [{
      boundBone {boneName=dyn_shirt_l}
      boxExtents=<15 20 10>
      constraintSetup {angularLimitsMax=<0 10 15>, angularLimitsMin=<0 -15 -15>, angularTarget=<0 0 0>}
      localJointOffset=<2 0 0>
     }]
   }
   n26 = Anim Dynamics - Bone: dyn_shirt_r!(ComponentPose=@n25, bAlphaBoolEnabled:bool=True, ExternalForce="0,0,0") -> (Pose) {
    alphaInputType=Bool
    angularDampingOverride=0.99
    angularSpringConstant=1000
    bAngularSpring=true
    bOverrideAngularDamping=true
    bUseGravityOverride=true
    simulationSpace=RootRelative
    alphaBoolBlend {blendOption=Linear}
    boundBone {boneName=dyn_shirt_r}
    componentAppliedLinearAccClamp=<5000 5000 5000>
    componentLinearAccScale=<0.1 0.35 0.35>
    componentLinearVelScale=<0.1 0.35 0.35>
    gravityOverride {z=-980}
    physicsBodyDefinitions [{
      boundBone {boneName=dyn_shirt_r}
      boxExtents=<15 20 10>
      constraintSetup {angularLimitsMax=<0 10 15>, angularLimitsMin=<0 -15 -15>, angularTarget=<0 0 0>}
      localJointOffset=<2 0 0>
     }]
   }
   n27 = Get thigh_rot_up_l() -> (thigh_rot_up_l:double)
   n28 = "Transform (Modify) Bone - Bone: dyn_skirt_fr_a_base_l"(ComponentPose=@n26, Alpha=@n27, Translation="6,0,0", Rotation="0,0,0", Scale="1,1,1") {
    rotationMode=BMM_Additive
    rotationSpace=BCS_BoneSpace
    translationMode=BMM_Additive
    translationSpace=BCS_BoneSpace
    alphaBoolBlend {blendOption=Linear}
    boneToModify {boneName=dyn_skirt_fr_a_base_l}
    translation {x=6}
   }
   n29 = "Transform (Modify) Bone - Bone: dyn_skirt_side_a_base_l"(ComponentPose=@n28, Alpha=@n27, Translation="1,-1,0", Rotation="0,-20,-5", Scale="1,1,1") {
    rotationMode=BMM_Additive
    rotationSpace=BCS_BoneSpace
    translationMode=BMM_Additive
    translationSpace=BCS_BoneSpace
    alphaBoolBlend {blendOption=Linear}
    boneToModify {boneName=dyn_skirt_side_a_base_l}
    rotation {roll=-5, yaw=-20}
    translation=<1 -1>
   }
   n30 = "Transform (Modify) Bone - Bone: dyn_skirt_bk_a_base_l"(ComponentPose=@n29, Alpha=@n27, Translation="0,2,1", Rotation="0,10,0.000", Scale="1,1,1") {
    rotationMode=BMM_Additive
    rotationSpace=BCS_BoneSpace
    translationMode=BMM_Additive
    translationSpace=BCS_BoneSpace
    alphaBoolBlend {blendOption=Linear}
    boneToModify {boneName=dyn_skirt_bk_a_base_l}
    rotation {yaw=10}
    translation {y=2, z=1}
   }
   n31 = Get thigh_rot_up_r() -> (thigh_rot_up_r:double)
   n32 = "Transform (Modify) Bone - Bone: dyn_skirt_fr_a_base_r"(ComponentPose=@n30, Alpha=@n31, Translation="6,0,0", Rotation="0,0,0", Scale="1,1,1") {
    rotationMode=BMM_Additive
    rotationSpace=BCS_BoneSpace
    translationMode=BMM_Additive
    translationSpace=BCS_BoneSpace
    alphaBoolBlend {blendOption=Linear}
    boneToModify {boneName=dyn_skirt_fr_a_base_r}
    translation {x=6}
   }
   n33 = "Transform (Modify) Bone - Bone: dyn_skirt_side_a_base_r"(ComponentPose=@n32, Alpha=@n31, Translation="0,0,0", Rotation="0,20,0", Scale="1,1,1") {
    rotationMode=BMM_Additive
    rotationSpace=BCS_BoneSpace
    translationMode=BMM_Additive
    translationSpace=BCS_BoneSpace
    alphaBoolBlend {blendOption=Linear}
    boneToModify {boneName=dyn_skirt_side_a_base_r}
    rotation {yaw=20}
   }
   n34 = "Transform (Modify) Bone - Bone: dyn_skirt_bk_a_base_r"(ComponentPose=@n33, Alpha=@n31, Translation="0,1,1", Rotation="0,-10,0.000", Scale="1,1,1") {
    rotationMode=BMM_Additive
    rotationSpace=BCS_BoneSpace
    translationMode=BMM_Additive
    translationSpace=BCS_BoneSpace
    alphaBoolBlend {blendOption=Linear}
    boneToModify {boneName=dyn_skirt_bk_a_base_r}
    rotation {yaw=-10}
    translation {y=1, z=1}
   }
   n35 = RigidBody!(ComponentPose=@n34, Alpha, OverrideWorldGravity:Vector="0,0,-980", ExternalForce) -> (Pose) {
    bClampLinearTranslationLimitToRefPose=true
    bOverrideWorldGravity=true
    overridePhysicsAsset=&/F_MED_Rumble_Dyn
    simulationSpace=BaseBoneSpace
    alphaBoolBlend {blendOption=Linear}
    baseBoneRef {boneName=spine_02}
    componentAppliedLinearAccClamp=<3000 3000 3000>
    componentLinearAccScale {x=0.4}
    componentLinearVelScale {x=0.4}
    overrideWorldGravity {z=-980}
    simSpaceSettings {
     maxAngularAcceleration=200
     maxAngularVelocity=5
     maxLinearAcceleration=1000
     maxLinearVelocity=200
     velocityScaleZ=0
     worldAlpha=1
     externalLinearDragV=<1 1>
    }
   }
   n36 = Get thigh_rot_otr_l()
   n37 = "Transform (Modify) Bone - Bone: dyn_skirt_fr_b_l"(ComponentPose=@n35, Alpha=@n36, Translation="0,0,0", Rotation="8,0,0", Scale="1,1,1") {
    rotationMode=BMM_Additive
    rotationSpace=BCS_BoneSpace
    translationMode=BMM_Additive
    translationSpace=BCS_BoneSpace
    alphaBoolBlend {blendOption=Linear}
    boneToModify {boneName=dyn_skirt_fr_b_l}
    rotation {pitch=8}
   }
   n38 = "Transform (Modify) Bone - Bone: dyn_skirt_side_b_l"(ComponentPose=@n37, Alpha=@n36, Translation="0,0,0", Rotation="8,0,0", Scale="1,1,1") {
    rotationMode=BMM_Additive
    rotationSpace=BCS_BoneSpace
    translationMode=BMM_Additive
    translationSpace=BCS_BoneSpace
    alphaBoolBlend {blendOption=Linear}
    boneToModify {boneName=dyn_skirt_side_b_l}
    rotation {pitch=8}
   }
   n39 = Get thigh_rot_inr_l() -> (thigh_rot_inr_l:double)
   n40 = "Transform (Modify) Bone - Bone: dyn_skirt_fr_a_l"(ComponentPose=@n38, Alpha=@n39, Translation="0,0,0", Rotation="5,10,0", Scale="1,1,1") {
    rotationMode=BMM_Additive
    rotationSpace=BCS_BoneSpace
    translationMode=BMM_Additive
    translationSpace=BCS_BoneSpace
    alphaBoolBlend {blendOption=Linear}
    boneToModify {boneName=dyn_skirt_fr_a_l}
    rotation {pitch=5, yaw=10}
   }
   n41 = "Transform (Modify) Bone - Bone: dyn_skirt_fr_a_r"(ComponentPose=@n40, Alpha=@n39, Translation="0,1,0", Rotation="5,0,0", Scale="1,1,1") {
    rotationMode=BMM_Additive
    rotationSpace=BCS_BoneSpace
    translationMode=BMM_Additive
    translationSpace=BCS_BoneSpace
    alphaBoolBlend {blendOption=Linear}
    boneToModify {boneName=dyn_skirt_fr_a_r}
    rotation {pitch=5}
    translation {y=1}
   }
   Output Pose(Result=Component To Local(ComponentPose=@n41)) {layerGroup=None}
   Get thigh_rot_inr_l()
  ]}
 EventGraph {nodes [
   Event BlueprintUpdateAnimation()
   n1 = BreakTransform(InTransform=GetDeltaTransformFromRefPose(self=GetOwnerSkeletalMeshComponent(), BoneName=thigh_r, BaseName=pelvis))
   > Set thigh_rot_up_r(thigh_rot_up_r:double=MapRangeClamped(Value=BreakRotator(InRot=@n1.Rotation).Yaw, InRangeA=-10, InRangeB=-100)) -> (Output_Get:double)
   n2 = GetOwnerSkeletalMeshComponent()
   n3 = BreakRotator(InRot=BreakTransform(InTransform=GetDeltaTransformFromRefPose(self=@n2, BoneName=thigh_l, BaseName=pelvis)).Rotation)
   > Set thigh_rot_up_l(thigh_rot_up_l:double=MapRangeClamped(Value=@n3.Yaw, InRangeA=-10, InRangeB=-100)) -> (Output_Get:double)
   > Set thigh_rot_inr_l(thigh_rot_inr_l:double=MapRangeClamped(Value=@n3.Pitch, InRangeA=0, InRangeB=-15)) -> (Output_Get:double)
   > Set thigh_rot_otr_l(thigh_rot_otr_l:double=MapRangeClamped(Value=@n3.Pitch, InRangeA=20, InRangeB=35)) -> (Output_Get:double)
   n4 = GetOwnerSkeletalMeshComponent()
   n5 = BreakRotator(InRot=BreakTransform(InTransform=GetDeltaTransformFromRefPose(self=@n4, BoneName=upperarm_l, BaseName=spine_05)).Rotation)
   > Set arm_rot_up_l(arm_rot_up_l:double=MapRangeClamped(Value="Absolute (Float)"(A=@n5.Yaw), InRangeA=50, InRangeB=130)) -> (Output_Get:double)
   Event OnCharacterCustomizationLoaded()
   > Do Once()
   > Set isWearBackbling(isWearBackbling:bool=CheckForBackbling()) -> (Output_Get:bool)
  ]}
]
variables [
 arm_rot_up_l {type=number}
 isWearBackbling {type=boolean}
 thigh_rot_inr_l {type=number}
 thigh_rot_otr_l {type=number}
 thigh_rot_up_l {type=number}
 thigh_rot_up_r {type=number}
]
